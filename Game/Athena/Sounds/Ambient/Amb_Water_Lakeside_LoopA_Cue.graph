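# Sound cue for outdoor ambient lakeside water in Athena, named as a loop and built on a lake shore lapping wave.
# A quality level node feeds a single wave player from its first slot, leaving the other three empty.
# It plays at reduced volume in the outdoor ambient sound class and references river water attenuation settings.
# It also carries attenuation override values and a concurrency override limiting it to one instance, preventing new ones.
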SoundCue /Game/Athena/Sounds/Ambient/Amb_Water_Lakeside_LoopA_Cue
root=QualityLevel_0
nodes [
 {id=QualityLevel_0, children [WavePlayer_2,null,null,null]}
 {id=WavePlayer_2, properties {soundWaveAssetPtr=&/Water_Lake_Shore_Lapping_01}}
]
properties {
 attenuationSettings=&/Fort_Amb_Water_River_Att
 maxDistance=2925
 soundClassObject=&Fort_Ambient_Outdoors_Athena
 volumeMultiplier=0.301
 attenuationOverrides {falloffDistance=2000, nonSpatializedRadiusStart=700, stereoSpread=1000, attenuationShapeExtents {x=925}}
 concurrencyOverrides {maxCount=1, resolutionRule=PreventNew}
}
