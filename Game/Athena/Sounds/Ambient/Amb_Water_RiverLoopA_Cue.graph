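# Looping river ambience cue for Athena outdoor ambience, spatialized with its own attenuation override that adds low-pass filtering over distance.
# An Intensity parameter crossfades a medium stream loop into a louder large waterfall loop as it rises from 0 to 1.
# A quality-level switch at the root uses that crossfade for the first two levels and a single optimized stream loop for the other two.
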SoundCue /Game/Athena/Sounds/Ambient/Amb_Water_RiverLoopA_Cue
root=QualityLevel_0
nodes [
 {
  id=ParamCrossFade_0
  children [WavePlayer_0,WavePlayer_2]
  properties {
   paramName=Intensity
   crossFadeInput [
    {fadeInDistanceEnd=0, fadeInDistanceStart=0, fadeOutDistanceEnd=1, fadeOutDistanceStart=0, volume=1}
    {fadeInDistanceEnd=1, fadeInDistanceStart=0, fadeOutDistanceEnd=100, fadeOutDistanceStart=100, volume=3}
   ]
  }
 }
 {id=QualityLevel_0, children [ParamCrossFade_0, ParamCrossFade_0, WavePlayer_3, WavePlayer_3]}
 {id=WavePlayer_0, properties {bLooping=true, soundWaveAssetPtr=/Game/Sounds/Fort_Ambient_Looping/Water/stream_medium_B_01}}
 {id=WavePlayer_2, properties {bLooping=true, soundWaveAssetPtr=/Game/Sounds/Fort_Ambient_Looping/Water/waterfall_large_A_01}}
 {id=WavePlayer_3, properties {bLooping=true, soundWaveAssetPtr=/Game/Sounds/Fort_Ambient_Looping/Water/stream_medium_B_01_opt}}
]
properties {
 attenuationSettings=&/Fort_Amb_Water_River_Att
 bOverrideAttenuation=true
 maxDistance=6925
 soundClassObject=&Fort_Ambient_Outdoors_Athena
 volumeMultiplier=0.301
 attenuationOverrides {
  bAttenuateWithLPF=true
  falloffDistance=6000
  lPFFrequencyAtMax=500
  nonSpatializedRadiusStart=700
  stereoSpread=1000
  attenuationShapeExtents {x=925}
 }
 concurrencyOverrides {maxCount=1, resolutionRule=PreventNew}
}
