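# Looping river water ambience cue for Athena, routed to the outdoor ambient sound class.
# A quality-level node selects one of four looping branches: the first two play stream_medium_B_02 directly.
# The last two play the stream_medium_B_01_opt wave through modulators that apply a fixed upward pitch shift.
# The cue plays at reduced volume and overrides its referenced river attenuation with inline falloff and spatialization settings.
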SoundCue /Game/Athena/Sounds/Ambient/Amb_Water_RiverLoopB_Cue
root=QualityLevel_0
nodes [
 {id=Modulator_0, children [WavePlayer_3], properties {pitchMax=1.2, pitchMin=1.2}}
 {id=Modulator_1, children [WavePlayer_4], properties {pitchMax=1.2, pitchMin=1.2}}
 {id=QualityLevel_0, children [WavePlayer_1,WavePlayer_5,Modulator_0,Modulator_1]}
 {id=WavePlayer_1, properties {bLooping=true, soundWaveAssetPtr=/Game/Sounds/Fort_Ambient_Looping/Water/stream_medium_B_02}}
 {id=WavePlayer_3, properties {bLooping=true, soundWaveAssetPtr=/Game/Sounds/Fort_Ambient_Looping/Water/stream_medium_B_01_opt}}
 {id=WavePlayer_4, properties {bLooping=true, soundWaveAssetPtr=&stream_medium_B_01_opt}}
 {id=WavePlayer_5, properties {bLooping=true, soundWaveAssetPtr=&stream_medium_B_02}}
]
properties {
 attenuationSettings=&/Fort_Amb_Water_River_Att
 bOverrideAttenuation=true
 maxDistance=2925
 soundClassObject=&Fort_Ambient_Outdoors_Athena
 volumeMultiplier=0.301
 attenuationOverrides {falloffDistance=2000, nonSpatializedRadiusStart=700, stereoSpread=512, attenuationShapeExtents {x=925}}
 concurrencyOverrides {maxCount=1, resolutionRule=PreventNew}
}
