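# Looping river ambience cue for Athena outdoors, playing a medium stream sound at reduced volume.
# A quality level node picks between looping wave players, using the full stream wave for the first two levels and its optimized variant for the last two.
# It references a river water attenuation asset but enables overridden attenuation, with stereo spread and a non-spatialized radius, and uses the outdoor ambient sound class.
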SoundCue /Game/Athena/Sounds/Ambient/Amb_Water_RiverLoopC_Cue
root=QualityLevel_0
nodes [
 {id=QualityLevel_0, children [WavePlayer_2,WavePlayer_6,WavePlayer_0,WavePlayer_7]}
 {id=WavePlayer_0, properties {bLooping=true, soundWaveAssetPtr=/Game/Sounds/Fort_Ambient_Looping/Water/stream_medium_A_03_opt}}
 {id=WavePlayer_2, properties {bLooping=true, soundWaveAssetPtr=/Game/Sounds/Fort_Ambient_Looping/Water/stream_medium_A_03}}
 {id=WavePlayer_6, properties {bLooping=true, soundWaveAssetPtr=&stream_medium_A_03}}
 {id=WavePlayer_7, properties {bLooping=true, soundWaveAssetPtr=&stream_medium_A_03_opt}}
]
properties {
 attenuationSettings=&/Fort_Amb_Water_River_Att
 bOverrideAttenuation=true
 maxDistance=2925
 soundClassObject=&Fort_Ambient_Outdoors_Athena
 volumeMultiplier=0.301
 attenuationOverrides {falloffDistance=2000, nonSpatializedRadiusStart=700, stereoSpread=1000, attenuationShapeExtents {x=925}}
 concurrencyOverrides {maxCount=1, resolutionRule=PreventNew}
}
