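; Fortnite ambient audio controller Blueprint that manages cosmetic-only weather and indoor ambience.
; BeginPlay adds the Curie bank and starts a timer running either a native update or the Blueprint Periodic Update.
; That update reads audio modulation buses to set the IsPlayerIndoors material parameter and delta-gate the indoor entry and tag.
; It also moves looping gameplay cues between view targets on supported levels, plus location-based snow audio, commented as disabled.
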(event EventBeginPlay
  (Networking|CanExecuteCosmeticEvents
    (:True
      (Utilities|FlowControl|Sequence
        (:then_0
          (AmbientAudio|AddAmbientEntry (WorldSubsystems|GetAmbientAudioSubsystem) "Curie" (Variables|Weather|GetCurieAudioBank) -1 1.0))
        (:then_1
          (Utilities|Time|SetTimerbyEvent (select (AmbientAudioController|ShouldUseNativeControllerUpdate) (EventDispatchers|CreateEvent :!event AmbientAudioControllerUpdateFromNative) (EventDispatchers|CreateEvent :!event Custom|Periodic%20Update)) (Utilities|Struct|BreakPerPlatformFloat (Variables|Default|GetPlatformFrequencyCheck)) true)
          (bind _returnvalue (Game|GetCurrentLevelName true))
          (Variables|Default|SetbTagApplicationEnabled (Utilities|Array|ContainsItem (Variables|Default|GetSupportedTagLevels) _returnvalue)))
        (:then_2
          (bind _gamestate (Class|FortAsyncActionGameStateReady|GameStateReadyAsync)
            (:OnReady
              (CallFunction|InitSnowCoverageComponent _gamestate))))))))

(event Custom|Periodic%20Update
  (bind _returnvalue (WorldSubsystems|GetAmbientAudioSubsystem))
  (Utilities|FlowControl|Sequence
    (:then_0
      ;[0 0 0] Handle persistent GCNL application for world pfx
      (if (Variables|Default|GetbTagApplicationEnabled)
        (CallFunction|CacheViewTarget)))
    (:then_1
      ;[0 0 0] DISABLED - dynamic snow support based on location
      (if (Variables|Snow|GetbSnowEnabled)
        (CallFunction|UpdateSnowSetupBP)))
    (:then_2
      ;[0 0 0] Update Indoors MPC scalar based on upward-firing traces
      (Rendering|Material|SetScalarParameterValue "/Game/Packages/Fortress_SharedMaterials/GlobalMaterialParameters/FortniteMaterialParameters.FortniteMaterialParameters" "IsPlayerIndoors" (Audio|Modulation|GetModulatorValue (Variables|Update|GetUpQuadrantControlBus)) :!class /Script/Engine.KismetMaterialLibrary))
    (:then_3
      (bind _updated_variable (Utilities|FlowControl|DeltaGateWildcard (Variables|InteriorDetection|GetbIsIndoors) (>= (Audio|Modulation|GetModulatorValue (Variables|Update|GetEnclosureControlBus)) (Variables|Update|GetIndoorsEnclosureThreshold)))
        (:Success
          (if _updated_variable
            (AmbientAudio|AddAmbientEntry _returnvalue "Indoor" (Variables|Update|GetInteriorAmbientAudioDataEntry) -1 0.1)
            (AmbientAudio|AddGameplayTag _returnvalue (Variables|Update|GetInteriorAmbientStateTag))
            (else
              (AmbientAudio|RemoveAmbientEntry _returnvalue "Indoor" 0.1)
              (AmbientAudio|RemoveGameplayTag _returnvalue (Variables|Update|GetInteriorAmbientStateTag)))))))))

(fn ApplyPlayerTags
  (params (Player :type Actor%20Object%20Reference) (Tag%20Array :type Array%20of%20Gameplay%20Tag%20Structures))
  (for _array_element Tag%20Array
    (Projectile|AddGameplayCue%28Looping%29Local Player _array_element (Ability|GameplayCue|MakeGameplayCueParameters 1.0 1.0))))

(fn CacheViewTarget
  (bind _currviewtarget (Variables|Default|GetCurrViewTarget))
  (bind _output_get (LocalVariables|SetNewViewTarget (Pawn|GetViewTarget (Fortnite|GetFirstLocalFortPlayerController true true))))
  (if (Utilities|NotEqual(Object) _output_get _currviewtarget)
    (CallFunction|OnViewTargetChanged _output_get _currviewtarget)
    (Variables|Default|SetCurrViewTarget (LocalVariables|GetNewViewTarget))))

(fn InitSnowCoverageComponent
  (params (Game%20State :type Game%20State%20Base%20Object%20Reference))
  (Variables|Default|SetSnowCoverageComponent (Actor|GetComponentbyClass Game%20State "/Script/FortniteGame.FortGameStateComponent_SurfaceOverride")))

(fn OnViewTargetChanged
  (params (New%20Target :type Actor%20Object%20Reference) (Prev%20Target :type Actor%20Object%20Reference))
  (bind _generictagstoapply (Variables|Default|GetGenericTagsToApply))
  (CallFunction|RemovePlayerTags Prev%20Target _generictagstoapply)
  (CallFunction|ApplyPlayerTags New%20Target _generictagstoapply))

(fn RemovePlayerTags
  (params (Player :type Actor%20Object%20Reference) (Array :type Array%20of%20Gameplay%20Tag%20Structures))
  (for _array_element Array
    (Projectile|RemoveGameplayCue%28Looping%29Local Player _array_element (Ability|GameplayCue|MakeGameplayCueParameters 0.0))))

(fn UpdateSnowSetupBP
  (bind _currviewtarget (Variables|Default|GetCurrViewTarget))
  (bind _returnvalue (Transformation|GetActorLocation _currviewtarget))
  (bind _returnvalue_2 (WorldSubsystems|GetAmbientAudioSubsystem))
  (bind _snowentryname (Variables|Snow|GetSnowEntryName))
  (bind _specialsurfacepadding (Variables|Default|GetSpecialSurfacePadding))
  (bind _issnowingsoundmix (Variables|Default|GetIsSnowingSoundMix))
  (Utilities|IsValid _currviewtarget
    (:Is%20Valid
      (bind _snowcoveragecomponent (Variables|Default|GetSnowCoverageComponent)
        (:then
          (if (and (Utilities|Enum|Equal(Enum) (Class|FortGameStateComponentSurfaceOverride|GetSpecialSurfaceType _snowcoveragecomponent) "SurfaceType23") (Math|Boolean|NotEqual(Boolean) (Variables|Snow|GetbInSnowCurrent) (SurfaceOverride|IsLocationinSpecialSurface _snowcoveragecomponent _returnvalue _specialsurfacepadding)))
            (bind _output_get (Variables|Snow|SetbInSnowCurrent (SurfaceOverride|IsLocationinSpecialSurface _snowcoveragecomponent _returnvalue _specialsurfacepadding)))
            (if _output_get
              (CallFunction|ApplyPlayerTags (Variables|Default|GetCurrViewTarget) (Variables|Snow|GetSnowTagsToApply))
              (AmbientAudio|AddAmbientEntry _returnvalue_2 _snowentryname (Variables|Weather|GetWinterAudioBank) -1 5.0)
              (Audio|PushSoundMixModifier _issnowingsoundmix)
              (else
                (CallFunction|RemovePlayerTags (Variables|Default|GetCurrViewTarget) (Variables|Snow|GetSnowTagsToApply))
                (AmbientAudio|RemoveAmbientEntry _returnvalue_2 _snowentryname -1.0)
                (Audio|PopSoundMixModifier _issnowingsoundmix)))))
        (:else)))))
