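; BP_AmbientLakeAudioActor is an ambient audio actor for lake water loop sound.
; Its construction script writes the actor's AttenuationExtents and FalloffDistance variables into its attenuation override struct and applies that to the WaterLoopAudio component.
; It then assigns the actor's Sound variable to that component, so sound and attenuation shape are configurable per placed instance.
; BeginPlay has no logic.
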(event EventBeginPlay)

(fn ConstructionScript
  (bind _structout (Utilities|Struct|SetmembersinSoundAttenuationSettings (Variables|Default|GetAttenuationOverride) (Variables|Default|GetAttenuationExtents) (Variables|Default|GetFalloffDistance)))
  (Variables|/Script/Engine.AudioComponent|SetAttenuationOverrides _structout (Variables|Default|GetWaterLoopAudio))
  (Audio|Components|Audio|SetSound (Variables|Default|GetWaterLoopAudio) (Variables|Default|GetSound)))
